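# Sound cue for the looping ambient sound of the CuddleFish, played at reduced volume.
# A single looping wave player feeds the cue root directly, with no mixing, randomization or modulation nodes.
# The cue uses a CuddleFish ambient attenuation asset, overrides concurrency to cap simultaneous instances at two, and routes through the Fort_Explosives sound class.
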SoundCue /PrimalGameplay/Sounds/CuddleFish/CuddleFish_AmbientLoop_Cue
root=WavePlayer_0
nodes [{id=WavePlayer_0, properties {bLooping=true, soundWaveAssetPtr=&/Cuddle_Fish_Amb_Loop}}]
properties {
 attenuationSettings=&/CuddleFish_Ambient_Att
 bOverrideConcurrency=true
 maxDistance=1024
 soundClassObject=/Game/Sounds/FortSoundClassesAndModes/Fort_Explosives
 volumeMultiplier=0.2
 attenuationOverrides {distanceAlgorithm=Logarithmic, falloffDistance=1024, stereoSpread=1, attenuationShapeExtents {x=150}}
 concurrencyOverrides {maxCount=2}
}
