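# Sound cue for the Cuddle Fish armed sound, classed as a ranged weapon projectile effect.
# A single wave player plays the armed wave through a default Modulator at reduced cue volume.
# It references the shared Cuddle Fish bounce attenuation asset and also carries override values for logarithmic distance falloff with low-pass filtering.
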SoundCue /PrimalGameplay/Sounds/CuddleFish/CuddleFish_Arm_Cue
root=Modulator_1
nodes [
 {id=Modulator_1, children [WavePlayer_0]}
 {id=WavePlayer_0, properties {soundWaveAssetPtr=&/Cuddle_Fish_Armed}}
]
properties {
 attenuationSettings=&/Fort_CuddleFish_Bounce_Att
 maxDistance=3000
 soundClassObject=&Fort_WeaponRanged_Projectile_Effects
 volumeMultiplier=0.25
 attenuationOverrides {
  bAttenuateWithLPF=true
  distanceAlgorithm=Logarithmic
  falloffDistance=2500
  lPFFrequencyAtMax=2000
  lPFRadiusMax=3000
  lPFRadiusMin=500
  attenuationShapeExtents {x=500}
 }
}
