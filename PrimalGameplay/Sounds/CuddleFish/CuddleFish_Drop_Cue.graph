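# Sound cue for the CuddleFish item drop, playing a cuddle fish landing sound.
# A single wave player feeds a modulator for per-play variation.
# It uses weapon-drop attenuation and the loot-collect sound class, at a reduced volume.
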SoundCue /PrimalGameplay/Sounds/CuddleFish/CuddleFish_Drop_Cue
root=Modulator_0
nodes [
 {id=Modulator_0, children [WavePlayer_1]}
 {id=WavePlayer_1, properties {soundWaveAssetPtr=&/Cuddle_Fish_Land_01}}
]
properties {
 attenuationSettings=&Fort_Weapon_Drop_Att
 maxDistance=3000
 soundClassObject=&Fort_Loot_Collect
 volumeMultiplier=0.3
}
