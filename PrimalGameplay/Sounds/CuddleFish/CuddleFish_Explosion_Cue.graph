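# CuddleFish explosion sound cue, routed to the large explosives sound class.
# It layers close and medium-distance recordings, each with its own projectile attenuation and pitch/volume modulator.
# The root quality switch gives the first level randomized wave variants, while the other three share a mix of one fixed wave per layer.
# Concurrency allows one instance per owner and blocks new plays while it is active.
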SoundCue /PrimalGameplay/Sounds/CuddleFish/CuddleFish_Explosion_Cue
root=QualityLevel_0
nodes [
 {
  id=Attenuation_0
  children [Modulator_13]
  properties {attenuationSettings=&Fort_Weapon_Projectile_Medium_Mono_Athena_Att}
 }
 {id=Attenuation_1, children [Modulator_12], properties {attenuationSettings=&Fort_Weapon_Projectile_Close_Athena_Att}}
 {
  id=Attenuation_2
  children [Modulator_15]
  properties {attenuationSettings=&Fort_Weapon_Projectile_Medium_Mono_Athena_Att}
 }
 {id=Attenuation_3, children [Modulator_14], properties {attenuationSettings=&Fort_Weapon_Projectile_Close_Athena_Att}}
 {id=Mixer_2, children [Attenuation_1, Attenuation_0], properties {inputVolume [1,1]}}
 {id=Mixer_5, children [Attenuation_3, Attenuation_2], properties {inputVolume [1,1]}}
 {id=Modulator_12, children [Random_1], properties {pitchMax=1, pitchMin=0.9}}
 {id=Modulator_13, children [QualityLevel_1], properties {pitchMax=1.15, pitchMin=0.85}}
 {id=Modulator_14, children [WavePlayer_5]}
 {id=Modulator_15, children [WavePlayer_7]}
 {id=QualityLevel_0, children [Mixer_2,Mixer_5,Mixer_5,Mixer_5]}
 {id=QualityLevel_1, children [Random_0,Random_0,WavePlayer_6,WavePlayer_6]}
 {id=Random_0, children [WavePlayer_3,WavePlayer_4], properties {weights [1,1]}}
 {id=Random_1, children [WavePlayer_1,WavePlayer_2], properties {weights [1,1]}}
 {id=WavePlayer_1, properties {soundWaveAssetPtr=&/Cuddle_Fish_Explo_01}}
 {id=WavePlayer_2, properties {soundWaveAssetPtr=&/Cuddle_Fish_Explo_02}}
 {id=WavePlayer_3, properties {soundWaveAssetPtr=&/Cuddle_Fish_Explo_MedDist_01}}
 {id=WavePlayer_4, properties {soundWaveAssetPtr=&/Cuddle_Fish_Explo_MedDist_02}}
 {id=WavePlayer_5, properties {soundWaveAssetPtr=&Cuddle_Fish_Explo_01}}
 {id=WavePlayer_6, properties {soundWaveAssetPtr=&Cuddle_Fish_Explo_MedDist_02}}
 {id=WavePlayer_7, properties {soundWaveAssetPtr=&Cuddle_Fish_Explo_MedDist_02}}
]
properties {
 bOverrideConcurrency=true
 maxDistance=10579.2871
 soundClassObject=/Game/Sounds/FortSoundClassesAndModes/Fort_Explosives_Large_PSM
 volumeMultiplier=0.85
 concurrencyOverrides {bLimitToOwner=true, maxCount=1, resolutionRule=PreventNew}
}
