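# Sound cue for the CuddleFish triggered jump, classed as a ranged weapon projectile effect.
# A single jump wave passes through a default-setting modulator into a mixer, whose second input is empty.
# It plays at reduced volume with the CuddleFish bounce attenuation.
# Overridden concurrency allows one instance per owner, stopping the oldest when retriggered.
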SoundCue /PrimalGameplay/Sounds/CuddleFish/CuddleFish_Jump_Triggered_Cue
root=Mixer_0
nodes [
 {id=Mixer_0, children [Modulator_1,null], properties {inputVolume [1,0.5]}}
 {id=Modulator_1, children [WavePlayer_1]}
 {id=WavePlayer_1, properties {soundWaveAssetPtr=&/Cuddle_Fish_Jump_01}}
]
properties {
 attenuationSettings=&/Fort_CuddleFish_Bounce_Att
 bOverrideConcurrency=true
 maxDistance=3000
 soundClassObject=&Fort_WeaponRanged_Projectile_Effects
 volumeMultiplier=0.2
 concurrencyOverrides {bLimitToOwner=true, maxCount=1, resolutionRule=StopOldest}
}
