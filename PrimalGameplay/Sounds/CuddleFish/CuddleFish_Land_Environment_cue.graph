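# Sound cue for the CuddleFish landing on the environment, playing one land wave through a Modulator node left at default settings.
# It plays at low volume in a third-person ranged weapon fire sound class.
# Distance falloff comes from the shared grenade bounce attenuation asset, and voice limiting from the CuddleFish attach concurrency set.
# Inline attenuation and concurrency override values are stored, but neither override is switched on.
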SoundCue /PrimalGameplay/Sounds/CuddleFish/CuddleFish_Land_Environment_cue.CuddleFish_Land_Environment_Cue
root=Modulator_0
nodes [
 {id=Modulator_0, children [WavePlayer_0]}
 {id=WavePlayer_0, properties {soundWaveAssetPtr=&/Cuddle_Fish_Land_01}}
]
properties {
 attenuationSettings=/Game/Sounds/Attenuation/Fort_Grenade_Bounce_Att
 maxDistance=5500
 soundClassObject=&Fort_WeaponRanged_Fire_3P_PSM
 volumeMultiplier=0.15
 attenuationOverrides {
  bAttenuateWithLPF=true
  distanceAlgorithm=Logarithmic
  falloffDistance=2500
  lPFFrequencyAtMax=2000
  lPFRadiusMax=3000
  lPFRadiusMin=500
  attenuationShapeExtents {x=500}
 }
 concurrencyOverrides {bLimitToOwner=true, maxCount=1, resolutionRule=PreventNew}
 concurrencySet [&/CuddleFish_Attach_Conc]
}
