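# Sound cue for a CuddleFish landing on a player, judging by its name, built from two Cuddle_Fish_Attach waves.
# A Random node picks one of the two with equal weight and feeds a Modulator left at default settings.
# It plays at reduced volume through the third-person ranged weapon fire sound class, borrowing the CuddleFish bounce attenuation.
# A shared CuddleFish attach concurrency asset governs overlapping plays.
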SoundCue /PrimalGameplay/Sounds/CuddleFish/CuddleFish_Land_Player_Cue
root=Modulator_0
nodes [
 {id=Modulator_0, children [Random_0]}
 {id=Random_0, children [WavePlayer_1,WavePlayer_2], properties {weights [1,1]}}
 {id=WavePlayer_1, properties {soundWaveAssetPtr=&/Cuddle_Fish_Attach_01}}
 {id=WavePlayer_2, properties {soundWaveAssetPtr=&/Cuddle_Fish_Attach_02}}
]
properties {
 attenuationSettings=&/Fort_CuddleFish_Bounce_Att
 maxDistance=3000
 priority=100
 soundClassObject=&Fort_WeaponRanged_Fire_3P_PSM
 volumeMultiplier=0.4
 attenuationOverrides {stereoSpread=1000}
 concurrencyOverrides {bLimitToOwner=true, maxCount=1, resolutionRule=PreventNew}
 concurrencySet [&/CuddleFish_Attach_Conc]
}
