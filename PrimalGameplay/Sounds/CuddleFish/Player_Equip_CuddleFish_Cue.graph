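# Sound cue for the player equipping the CuddleFish item, classed as player foley.
# A single wave player feeds the Cuddle_Fish_Equip_01 sound through a modulator at the root.
# It uses the generic player equip attenuation and overrides concurrency so new instances are prevented once the limit is reached.
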SoundCue /PrimalGameplay/Sounds/CuddleFish/Player_Equip_CuddleFish_Cue
root=Modulator_0
nodes [
 {id=Modulator_0, children [WavePlayer_1]}
 {id=WavePlayer_1, properties {soundWaveAssetPtr=&/Cuddle_Fish_Equip_01}}
]
properties {
 attenuationSettings=/Game/Sounds/Attenuation/Fort_Player_Equip_Generic_Att
 bOverrideConcurrency=true
 maxDistance=1800
 soundClassObject=&Fort_Player_Foley
 concurrencyOverrides {resolutionRule=PreventNew}
}
